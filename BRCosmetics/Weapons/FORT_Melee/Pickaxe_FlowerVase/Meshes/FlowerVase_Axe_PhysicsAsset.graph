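# Physics asset for the FlowerVase pickaxe cosmetic, a two-handed melee weapon mesh.
# It has a single body on the Melee_TwoHanded_Handle bone, made of one capsule that is offset and rotated relative to that bone.
# There are no constraints, so the pickaxe is simulated as one rigid piece.
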PhysicsAsset /BRCosmetics/Weapons/FORT_Melee/Pickaxe_FlowerVase/Meshes/FlowerVase_Axe_PhysicsAsset
bodies {Melee_TwoHanded_Handle [{shapeType=Capsule, length=48.8224, radius=10.5609, center=<0.016 4.1386 31.9242>, rotation=<rot 7.0341 -90 0.034>}]}
constraints []
